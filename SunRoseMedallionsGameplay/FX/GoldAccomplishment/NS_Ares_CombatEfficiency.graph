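# Niagara muzzle-flash-style fire burst for Ares Combat Efficiency in the SunRoseMedallions gold accomplishment FX.
# Short-lived sprite bursts form flames, flare, glow with dark halo, muzzle heat and curl-noise sparks, plus a mesh shockwave, all tinted by HDR color curves.
# System flags for distant muzzle FX and third-person viewing scale size and alpha, and time of day dims flames and flare at night.
# ADS scope-hiding modules are present but disabled.
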NiagaraSystem /SunRoseMedallionsGameplay/FX/GoldAccomplishment/NS_Ares_CombatEfficiency
dependencies {
 usedDataInterfaces [
  &NiagaraDataInterfaceCurve
  &NiagaraDataInterfaceVector2DCurve
  &NiagaraDataInterfaceColorCurve
  &NiagaraDataInterfaceCamera
  &NiagaraDataInterfaceSpriteRendererInfo
 ]
 usedDynamicInputs [&RandomRangeVector, &Boolean_AndOperation]
 usedModules [
  &self:SystemSpawnScript.SystemScriptSource.SystemScriptGraph.NiagaraNodeAssignment_0.SetVariables_#1
  &self:SystemSpawnScript.SystemScriptSource.SystemScriptGraph.NiagaraNodeAssignment_1.SetVariables_#2
  &self:SystemSpawnScript.SystemScriptSource.SystemScriptGraph.NiagaraNodeAssignment_2.SetVariables_#3
  &self:SystemSpawnScript.SystemScriptSource.SystemScriptGraph.NiagaraNodeAssignment_3.SetVariables_#4
  &self:SystemSpawnScript.SystemScriptSource.SystemScriptGraph.NiagaraNodeAssignment_5.SetVariables_#5
  &SystemState
 ]
}
emitters {
 Flames {
  emitterUpdateScript [
   &EmitterState()
   &SpawnBurst_Instantaneous(
    Spawn Count=&RandomRangeInt(Minimum=4, Maximum=8)
    Use Spawn Probability=false
    UseLoopCountLimit=false
   )
  ]
  particleSpawnScript [
   &InitializeParticle(
    Color Maximum=<rgba 1 0.4267 0.2 1>
    Color Minimum=<rgba 1 0.2 0.2267 0.3>
    Color Mode=::NewEnumerator2
    Lifetime Max=0.3
    Lifetime Min=0.1
    Lifetime Mode=::NewEnumerator1
    Sprite Size=&Multiply_Vector2D_ByFloat(
     Float=&RandomRangeFloat(Minimum=10, Maximum=20)
     Vector2D=&RandomRangeVector2D(Minimum=<0.3 1>, Maximum=<0.7 2>)
    )
    Sprite Size Mode=::NewEnumerator4
    UsePositionOffset=false
    Write Material Random=true
   )
   &ShapeLocation(MaskSpawnGroup=false, Override Seed=false, Sphere Radius=4, UseOwnerScale=true)
   &AddVelocity(
    Module.Use Falloff Distance=false
    Origin Offset=<-15 0 0>
    Override Seed=false
    Velocity Mode=::NewEnumerator1
    Velocity Speed=&RandomRangeFloat(Minimum=200, Maximum=500)
   )
   &InheritVelocity(Inherited Velocity Speed Limit=3000, Speed Limit=true, UseSpeedThreshold=false)
   &SubUVAnimation(
    "Sprite Renderer "=DataInterface() {propertyValues {SpriteRenderer=&self:Flames_1.NiagaraSpriteRendererProperties_0}}
    SubUV Animation Mode:ENiagara_SubUVLookupModeV2=::NewEnumerator2
    UseEndFrame=false
    UseStartFrame=false
   )
   &Color(Color=&MultiplyLinearColor(
     Linear Color=$Particles.Color
     Scale Factor=&MakeVector4_FromFloat(Float=&MakeCustomFloatFromBool(Bool=$System.ADSScopeHideFX, True Float=0, False Float=1))
    )) {enabled=false}
   /SetVariables(Particles.Color=&MultiplyLinearColor(
     Linear Color=$Particles.Color
     Scale Factor=&MakeFloat4FromVectorAndFloat(
      "Vector (RGB)"=<1 1 1>
      "Float (Alpha)"=&FNTimeOfDayScalar(Night Scale=0.8, Morning Scale=1, Evening Scale=1)
     )
    )) {isSetParameters=true}
  ]
  particleUpdateScript [
   &ParticleState()
   /SetVariables(Particles.ADSAlpha:NiagaraFloat=&RemapRange(
     Input Value To Remap=&VectorLength(Vector=&Subtract_Vector(
       A=&ReturnCameraPosition(Coordinate Space:ENiagaraCoordinateSpace=::World, Camera Query:NiagaraDataInterfaceCamera=DataInterface())
       B=$Particles.ShapeLocation.ShapePosition
      ))
     Input Min=50
     Input Max=100
     Output Min=1
     Output Max=0
    )) {enabled=false, isSetParameters=true}
   /SetVariables(Particles.ADSshow:NiagaraBool=&SetBoolByFloatComparison(A=$Particles.ADSAlpha, B=0.01, Comparison Type=::NewEnumerator7)) {enabled=false, isSetParameters=true}
   &ScaleSpriteSize(Uniform Curve Sprite Scale=DataInterface() {propertyValues {Curve {keys [
        {interpMode=RCIM_Cubic, tangentMode=RCTM_User, time=0, value=0}
        {interpMode=RCIM_Cubic, tangentMode=RCTM_User, time=0.2, value=1.6}
        {interpMode=RCIM_Cubic, tangentMode=RCTM_User, time=0.6, value=1}
       ]}}})
   &Drag(Drag=2, Use Linear Drag=true, Use Rotational Drag=true)
   &SolveForcesAndVelocity(Clamp Velocity=false, Limit Acceleration=false)
   &ScaleSpriteSizeBySpeed(Max Scale Factor=<1 3>, Velocity Threshold=400)
   &ScaleColor(
    Scale Alpha=&FloatFromCurve(
     FloatCurve=DataInterface() {propertyValues {Curve {keys [
         {interpMode=RCIM_Cubic, tangentMode=RCTM_User, time=0, value=0}
         {interpMode=RCIM_Cubic, tangentMode=RCTM_User, time=0.1, value=1}
         {interpMode=RCIM_Cubic, tangentMode=RCTM_User, time=1, value=0}
        ]}}}
     Scale Curve=&MakeCustomFloatFromBool(Bool=$System.ShowDistantMuzzleFX, True Float=1, False Float=0)
    )
    Scale RGB=&MakeVectorFromLinearColorRGB(Color=&ColorFromCurve(ColorCurve=DataInterface() {propertyValues {
        AlphaCurve {keys []}
        BlueCurve {keys [<at 0.1198 3.4 RCIM_Cubic>,<at 0.386 0.6>,<at 0.9697 0.16 RCIM_Cubic>]}
        GreenCurve {keys [<at 0.1198 9.2933 RCIM_Cubic>,<at 0.386 0.9533>,<at 0.9697 0.1707 RCIM_Cubic>]}
        RedCurve {keys [<at 0.1198 17 RCIM_Cubic>,<at 0.386 2>,<at 0.9697 0.8 RCIM_Cubic>]}
       }}))
    ScaleA=true
    ScaleRGB=true
   )
  ]
  rendererClasses [&NiagaraSpriteRendererProperties]
  renderers [{rendererClass=&NiagaraSpriteRendererProperties}]
 }
 Flare {
  emitterUpdateScript [
   &EmitterState()
   &SpawnBurst_Instantaneous(Use Spawn Probability=false, UseLoopCountLimit=false)
  ]
  particleSpawnScript [
   &InitializeParticle(
    Color=&MakeLinearColorFromVectorAndFloat(
     "Vector (RGB)"=<1 1 1>
     "Float (Alpha)"=&FNTimeOfDayScalar(Night Scale=0.7, Morning Scale=1, Evening Scale=1)
    )
    Lifetime=&RandomRangeFloat(Minimum=0.12, Maximum=0.17)
    Sprite Size=&Multiply_Vector2D_ByFloat(
     Float=&Multiply_Float(
      A=&RandomRangeFloat(Minimum=1.4, Maximum=1.9)
      B=&MakeCustomFloatFromBool(Bool=$System.ShowDistantMuzzleFX, True Float=1, False Float=1.2)
     )
     Vector2D=<85 60>
    )
    Sprite Size Mode=::NewEnumerator4
    UsePositionOffset=false
    Write Material Random=true
   )
   &CameraOffset(Camera Offset Amount=5)
   &SystemLocation(MaskSpawnGroup=false)
   &Color(Color=&MultiplyLinearColor(
     Linear Color=$Particles.Color
     Scale Factor=&MakeVector4_FromFloat(Float=&MakeCustomFloatFromBool(Bool=$System.ADSScopeHideFX, True Float=0, False Float=1))
    )) {enabled=false}
  ]
  particleUpdateScript [
   &ParticleState()
   &ScaleSpriteSize(
    Non-Uniform Curve Sprite Scale=DataInterface() {propertyValues {
      XCurve {keys [
        {arriveTangent=0.2895, interpMode=RCIM_Cubic, leaveTangent=0.2895, tangentMode=RCTM_User, time=0.25, value=0.6}
        {arriveTangent=1.9796, interpMode=RCIM_Cubic, leaveTangent=1.9796, tangentMode=RCTM_User, time=0.8, value=1.5}
       ]}
      YCurve {keys [
        {arriveTangent=-1.6077, interpMode=RCIM_Cubic, leaveTangent=-1.6077, tangentMode=RCTM_User, time=0.05, value=0.7}
        {arriveTangent=-0.1814, interpMode=RCIM_Cubic, leaveTangent=-0.1814, tangentMode=RCTM_User, time=0.6, value=0.45}
       ]}
     }}
    Scale Sprite Size Mode=::NewEnumerator3
   )
   &ScaleColor(
    Scale Alpha=&FloatFromCurve(
     FloatCurve=DataInterface() {propertyValues {Curve {keys [
         {interpMode=RCIM_Cubic, tangentMode=RCTM_User, time=0.02, value=1}
         {interpMode=RCIM_Cubic, tangentMode=RCTM_User, time=0.25, value=1}
         {arriveTangent=-0.2576, interpMode=RCIM_Cubic, leaveTangent=-0.2576, tangentMode=RCTM_User, time=1, value=0}
        ]}}}
     Scale Curve=&MakeCustomFloatFromBool(Bool=$System.ShowDistantMuzzleFX, True Float=1, False Float=0.05)
    )
    Scale RGB=&MakeVectorFromLinearColorRGB(Color=&ColorFromCurve(ColorCurve=DataInterface() {propertyValues {
        AlphaCurve {keys []}
        BlueCurve {keys [<at 0.0146 0.54 RCIM_Cubic>,<at 0.2992 0>,<at 1 0 RCIM_Cubic>]}
        GreenCurve {keys [<at 0.0146 4.95 RCIM_Cubic>,<at 0.2992 0.2333>,<at 1 0 RCIM_Cubic>]}
        RedCurve {keys [<at 0.0146 27 RCIM_Cubic>,<at 0.2992 7>,<at 1 0.8 RCIM_Cubic>]}
       }}))
    ScaleA=true
    ScaleRGB=true
   )
   /SetVariables(Particles.VisibilityTag=&MakeCustomIntFromBool(Bool=$System.ShowDistantMuzzleFX, True INT=0, False INT=1)) {isSetParameters=true}
   &ScaleColor(
    Scale Alpha=&MakeCustomFloatFromBool(Bool=$System.IsThirdPersonPlayerWatching, True Float=5, False Float=1)
    ScaleA=true
    ScaleRGB=true
   ) {module=ScaleColor001}
   &ScaleSpriteSize(
    Scale Sprite Size Mode=::NewEnumerator0
    Uniform Scale Factor=&MakeCustomFloatFromBool(Bool=$System.IsThirdPersonPlayerWatching, True Float=1.5, False Float=1)
   ) {module=ScaleSpriteSize001}
  ]
  rendererClasses [&NiagaraSpriteRendererProperties]
  renderers [
   {rendererClass=&NiagaraSpriteRendererProperties}
   {rendererClass=&NiagaraSpriteRendererProperties, rendererIndex=1}
  ]
 }
 Glow {
  emitterUpdateScript [
   &EmitterState()
   &SpawnBurst_Instantaneous(Use Spawn Probability=false, UseLoopCountLimit=false)
  ]
  particleSpawnScript [
   &InitializeParticle(
    Lifetime=&RandomRangeFloat(Minimum=0.17, Maximum=0.22)
    Sprite Size=&Multiply_Vector2D_ByFloat(
     Float=0.5
     Vector2D=&Multiply_Vector2D(
      A=<100 100>
      B=&MakeCustomVector2DFromBool(Bool=$System.ShowDistantMuzzleFX, True Vector2D=<1 1>, False Vector2D=<1 0.4>)
     )
    )
    Sprite Size Mode=::NewEnumerator4
    UsePositionOffset=false
    Write Material Random=true
   )
   &CameraOffset(Camera Offset Amount=12)
   &SystemLocation(MaskSpawnGroup=false)
   /SetVariables(Particles.SpriteSize=&Multiply_Vector2D_ByFloat(
     Float=&MakeCustomFloatFromBool(Bool=$System.IsThirdPersonPlayerWatching, True Float=1.5, False Float=0.8)
     Vector2D=$Particles.SpriteSize
    )) {isSetParameters=true}
   /SetVariables(Particles.SpriteSizeDarkGlow:Vector2f=&Multiply_Vector2D_ByFloat(Float=1.37, Vector2D=$Particles.SpriteSize)) {isSetParameters=true}
   &Color(Color=&MultiplyLinearColor(
     Linear Color=$Particles.Color
     Scale Factor=&MakeVector4_FromFloat(Float=&MakeCustomFloatFromBool(Bool=$System.ADSScopeHideFX, True Float=0, False Float=1))
    )) {enabled=false}
  ]
  particleUpdateScript [
   &ParticleState()
   &ScaleColor(
    Scale Alpha=&FloatFromCurve(
     FloatCurve=DataInterface() {propertyValues {Curve {keys [<at 0 1>,<at 0.7 0>]}}}
     Scale Curve=&MakeCustomFloatFromBool(Bool=$System.ShowDistantMuzzleFX, True Float=0.8, False Float=0.3)
    )
    Scale RGB=&MakeVectorFromLinearColorRGB(Color=&ColorFromCurve(ColorCurve=DataInterface() {propertyValues {
        AlphaCurve {keys []}
        BlueCurve {keys [<at 0.0477 0.21 RCIM_Cubic>,<at 0.3006 0.12>,<at 1 0 RCIM_Cubic>]}
        GreenCurve {keys [<at 0.0477 0.768 RCIM_Cubic>,<at 0.3006 0.208>,<at 1 0 RCIM_Cubic>]}
        RedCurve {keys [<at 0.0477 3 RCIM_Cubic>,<at 0.3006 1>,<at 1 1 RCIM_Cubic>]}
       }}))
    ScaleA=true
    ScaleRGB=true
   )
   /SetVariables(Particles.DarkGlowColor:LinearColor=&MakeLinearColorFromVectorAndFloat(
     "Vector (RGB)"=<0.01 0 0>
     "Float (Alpha)"=&FloatFromCurve(
      FloatCurve=DataInterface() {propertyValues {Curve {keys [<at 0 1>,<at 1 0>]}}}
      Scale Curve=&MakeCustomFloatFromBool(Bool=$System.ShowDistantMuzzleFX, True Float=2.7, False Float=1)
     )
    )) {isSetParameters=true}
   /SetVariables(Particles.DarkGlowColor:LinearColor=&MultiplyLinearColor(
     Linear Color=$Particles.DarkGlowColor
     Scale Factor=&MakeFloat4FromVectorAndFloat(
      "Vector (RGB)"=<1 1 1>
      "Float (Alpha)"=&MakeCustomFloatFromBool(Bool=$System.IsThirdPersonPlayerWatching, True Float=2, False Float=1)
     )
    )) {isSetParameters=true}
   &ScaleColor(
    Scale Alpha=&MakeCustomFloatFromBool(Bool=$System.IsThirdPersonPlayerWatching, True Float=3, False Float=1)
    ScaleA=true
    ScaleRGB=false
   ) {module=ScaleColor001}
  ]
  rendererClasses [&NiagaraSpriteRendererProperties]
  renderers [
   {rendererClass=&NiagaraSpriteRendererProperties}
   {rendererClass=&NiagaraSpriteRendererProperties, rendererIndex=1}
  ]
 }
 MuzzleHeat {
  emitterUpdateScript [
   &EmitterState()
   &SpawnBurst_Instantaneous(Use Spawn Probability=false, UseLoopCountLimit=false)
  ]
  particleSpawnScript [
   &InitializeParticle(Lifetime=0.27, Sprite Size Mode=::NewEnumerator3, Uniform Sprite Size=30, UsePositionOffset=false, Write Material Random=true)
   &CameraOffset(Camera Offset Amount=2)
   &SystemLocation(MaskSpawnGroup=false, Offset=<-2 0 0>)
   &Color(Color=&MultiplyLinearColor(
     Linear Color=$Particles.Color
     Scale Factor=&MakeVector4_FromFloat(Float=&MakeCustomFloatFromBool(Bool=$System.ADSScopeHideFX, True Float=0, False Float=1))
    )) {enabled=false}
  ]
  particleUpdateScript [
   /SetVariables(Particles.ADSAlpha:NiagaraFloat=&RemapRange(
     Input Value To Remap=&VectorLength(Vector=&Subtract_Vector(
       A=&ReturnCameraPosition(Coordinate Space:ENiagaraCoordinateSpace=::World, Camera Query:NiagaraDataInterfaceCamera=DataInterface())
       B=Linked() {linkedVariable {name=Particles.Position, type=NiagaraPosition}}
      ))
     Input Min=50
     Input Max=100
     Output Min=1
     Output Max=0
    )) {enabled=false, isSetParameters=true}
   &ParticleState()
   &ScaleColor(
    Scale Alpha=&FloatFromCurve(FloatCurve=DataInterface() {propertyValues {Curve {keys [
         {arriveTangent=-2.7708, interpMode=RCIM_Cubic, leaveTangent=-2.7708, tangentMode=RCTM_User, time=0.13, value=1}
         {interpMode=RCIM_Cubic, tangentMode=RCTM_User, time=1, value=0}
        ]}}})
    Scale RGB=&MakeVectorFromLinearColorRGB(Color=&ColorFromCurve(ColorCurve=DataInterface() {propertyValues {
        AlphaCurve {keys []}
        BlueCurve {keys [<at 0.0199 4.8 RCIM_Cubic>,<at 0.6723 0.07>,<at 1 0 RCIM_Cubic>]}
        GreenCurve {keys [<at 0.0199 17.7067 RCIM_Cubic>,<at 0.6723 0.763>,<at 1 0.0167 RCIM_Cubic>]}
        RedCurve {keys [<at 0.0199 40 RCIM_Cubic>,<at 0.6723 7>,<at 1 1 RCIM_Cubic>]}
       }}))
    ScaleA=true
    ScaleRGB=true
   )
  ]
  rendererClasses [&NiagaraSpriteRendererProperties]
  renderers [{rendererClass=&NiagaraSpriteRendererProperties}]
 }
 ShockWave {
  emitterUpdateScript [
   &EmitterState(Life Cycle Mode=::NewEnumerator1, Loop Behavior=::NewEnumerator1, Loop Delay=0.05, Loop Duration=1, UseLoopDelay=true)
   &SpawnRate(SpawnRate=0.8, Use Spawn Probability=false) {enabled=false}
   &SpawnBurst_Instantaneous(Use Spawn Probability=false, UseLoopCountLimit=false)
  ]
  particleSpawnScript [
   &InitializeParticle(
    Color=<rgba 1 1 1 0.3>
    Lifetime=&RandomRangeFloat(Minimum=0.15, Maximum=0.17)
    Mesh Scale=&Multiply_VectorByFloat(
     Vector=&MakeCustomVectorFromBool(Bool=$System.ShowDistantMuzzleFX, True Vector=<1 1 1>, False Vector=<0.6 0.6 0.3>)
     Float=0.8
    )
    Mesh Scale Mode=::NewEnumerator4
    UsePositionOffset=false
    Write Material Random=true
   )
   &SystemLocation(MaskSpawnGroup=false, Offset=$System.PSystemPosition)
   &CameraOffset(Camera Offset Amount=10)
   &Color(Color=&MultiplyLinearColor(
     Linear Color=$Particles.Color
     Scale Factor=&MakeVector4_FromFloat(Float=&MakeCustomFloatFromBool(Bool=$System.ADSScopeHideFX, True Float=0, False Float=1))
    )) {enabled=false}
  ]
  particleUpdateScript [
   &ParticleState()
   &ScaleMeshSize(Scale Factor=&VectorFromFloat(Value=&FloatFromCurve(
      FloatCurve=DataInterface() {propertyValues {Curve {keys [
          {arriveTangent=0.6176, interpMode=RCIM_Cubic, leaveTangent=0.6176, tangentMode=RCTM_User, time=0, value=0.75}
          {arriveTangent=0.0264, interpMode=RCIM_Cubic, leaveTangent=0.0264, tangentMode=RCTM_User, time=1, value=1}
         ]}}}
      Scale Curve=&MakeCustomFloatFromBool(Bool=$System.IsThirdPersonPlayerWatching, True Float=1.5, False Float=1)
     )))
   &ScaleColor(
    Scale Alpha=&FloatFromCurve(
     FloatCurve=DataInterface() {propertyValues {Curve {keys [
         {arriveTangent=13.3683, interpMode=RCIM_Cubic, leaveTangent=13.3683, tangentMode=RCTM_User, time=0, value=0}
         {arriveTangent=-2.8108, interpMode=RCIM_Cubic, leaveTangent=-2.8108, tangentMode=RCTM_User, time=0.1, value=1}
         {interpMode=RCIM_Cubic, tangentMode=RCTM_User, time=1, value=0}
        ]}}}
     Scale Curve=&MakeCustomFloatFromBool(Bool=$System.ShowDistantMuzzleFX, True Float=0.8, False Float=0)
    )
    Scale RGB=&MakeVectorFromLinearColorRGB(Color=&ColorFromCurve(ColorCurve=DataInterface() {propertyValues {
        AlphaCurve {keys []}
        BlueCurve {keys [<at 0.0212 0.18 RCIM_Cubic>,<at 0.1721 0>,<at 0.29 0.01>,<at 0.7393 0.03 RCIM_Cubic>]}
        GreenCurve {keys [<at 0.0212 0.744 RCIM_Cubic>,<at 0.1721 0.25>,<at 0.29 0.1255>,<at 0.7393 0.03 RCIM_Cubic>]}
        RedCurve {keys [<at 0.0212 3 RCIM_Cubic>,<at 0.1721 1>,<at 0.29 1>,<at 0.7393 1 RCIM_Cubic>]}
       }}))
    ScaleA=true
    ScaleRGB=true
   )
   &DynamicMaterialParameters(
    Index 0 Param 1=&FloatFromCurve(
     FloatCurve=DataInterface() {propertyValues {Curve {keys [
         {interpMode=RCIM_Cubic, tangentMode=RCTM_User, time=0, value=1}
         {arriveTangent=-0.3394, interpMode=RCIM_Cubic, leaveTangent=-0.3394, tangentMode=RCTM_User, time=0.3, value=3}
         {arriveTangent=-3.5082, interpMode=RCIM_Cubic, leaveTangent=-3.5082, tangentMode=RCTM_User, time=1, value=2}
        ]}}}
     Scale Curve=0.5
    )
    Module.Index0Param0Write=true
    Module.Index0Param1Write=false
    Module.Index0Param2Write=false
    Module.Index0Param3Write=true
   )
   &ScaleColor(
    Scale Alpha=&MakeCustomFloatFromBool(Bool=$System.IsThirdPersonPlayerWatching, True Float=1.3, False Float=1)
    ScaleA=true
    ScaleRGB=false
   ) {enabled=false, module=ScaleColor001}
  ]
  rendererClasses [&NiagaraMeshRendererProperties]
  renderers [{rendererClass=&NiagaraMeshRendererProperties, rendererIndex=0}]
 }
 Sparks {
  emitterUpdateScript [
   &EmitterState()
   &SpawnBurst_Instantaneous(Spawn Count=10, Use Spawn Probability=false, UseLoopCountLimit=false)
  ]
  particleSpawnScript [
   &InitializeParticle(
    Color=&RandomRangeLinearColor(Minimum=<rgba 1 0.9876 0.439 1>, Maximum=<rgba 1 0.4478 0.4578 1>)
    Lifetime Max=0.3
    Lifetime Min=0.1
    Lifetime Mode=::NewEnumerator1
    Sprite Size=&Multiply_Vector2D_ByFloat(Float=&RandomRangeFloat(Minimum=0.5, Maximum=1), Vector2D=<15 15>)
    Sprite Size Mode=::NewEnumerator4
    UsePositionOffset=false
    Write Material Random=true
   )
   &ShapeLocation(MaskSpawnGroup=false, Override Seed=false, Sphere Radius=5, UseOwnerScale=true)
   &AddVelocity(
    Module.Use Falloff Distance=false
    Origin Offset=<-7 0 0>
    Override Seed=false
    Velocity Mode=::NewEnumerator1
    Velocity Speed=&RandomRangeFloat(Minimum=300, Maximum=700)
   )
   &DynamicMaterialParameters(
    Index 0 Param 1=10
    Index 0 Param 2=1
    Module.Index0Param0Write=true
    Module.Index0Param1Write=true
    Module.Index0Param2Write=false
    Module.Index0Param3Write=false
   )
   &InheritVelocity(Inherited Velocity Speed Limit=3000, Speed Limit=true, UseSpeedThreshold=false)
   &Color(Color=&MultiplyLinearColor(
     Linear Color=$Particles.Color
     Scale Factor=&MakeVector4_FromFloat(Float=&MakeCustomFloatFromBool(Bool=$System.ADSScopeHideFX, True Float=0, False Float=1))
    )) {enabled=false}
  ]
  particleUpdateScript [
   &ParticleState()
   &ScaleSpriteSize(Uniform Curve Sprite Scale=DataInterface() {propertyValues {Curve {keys [{tangentMode=RCTM_User, time=0, value=1}, {tangentMode=RCTM_User, time=1, value=0}]}}})
   /SetVariables(Particles.ADSAlpha:NiagaraFloat=&RemapRange(
     Input Value To Remap=&VectorLength(Vector=&Subtract_Vector(
       A=&ReturnCameraPosition(Coordinate Space:ENiagaraCoordinateSpace=::World, Camera Query:NiagaraDataInterfaceCamera=DataInterface())
       B=$Particles.ShapeLocation.ShapePosition
      ))
     Input Min=50
     Input Max=100
     Output Min=1
     Output Max=0
    )) {enabled=false, isSetParameters=true}
   /SetVariables(Particles.ADSshow:NiagaraBool=&SetBoolByFloatComparison(A=$Particles.ADSAlpha, B=0.01, Comparison Type=::NewEnumerator7)) {enabled=false, isSetParameters=true}
   &CurlNoiseForce(
    Noise Frequency=200
    Noise Strength=&Multiply_Float(
     A=&RandomRangeFloat(Minimum=500, Maximum=700)
     B=&FloatFromCurve(FloatCurve=DataInterface() {propertyValues {Curve {keys [<at 0 0>,<at 0.1 1>]}}})
    )
    Pan Noise=true
    Pan Noise Field=<0.1 0.1 0.1>
   )
   &Drag(Drag=&RandomRangeFloat(Minimum=10, Maximum=13), Use Linear Drag=true, Use Rotational Drag=true)
   &SolveForcesAndVelocity(Clamp Velocity=false, Limit Acceleration=false)
   &ScaleSpriteSizeBySpeed(Max Scale Factor=<0.5 2>, Min Scale Factor=<0.7 1>, Velocity Threshold=120)
   &ScaleColor(
    Scale Alpha=&FloatFromCurve(
     FloatCurve=DataInterface() {propertyValues {Curve {keys [
         {interpMode=RCIM_Cubic, tangentMode=RCTM_User, time=0, value=1}
         {arriveTangent=-2.9167, interpMode=RCIM_Cubic, leaveTangent=-2.9167, tangentMode=RCTM_User, time=1, value=0}
        ]}}}
     Scale Curve=&MakeCustomFloatFromBool(Bool=$System.ShowDistantMuzzleFX, True Float=1, False Float=0)
    )
    Scale RGB=&MakeVectorFromLinearColorRGB(Color=&ColorFromCurve(ColorCurve=DataInterface() {propertyValues {
        AlphaCurve {keys []}
        BlueCurve {keys [<at 0.0514 0.96 RCIM_Cubic>,<at 0.67 0.04>,<at 0.984 0 RCIM_Cubic>]}
        GreenCurve {keys [<at 0.0514 2.016 RCIM_Cubic>,<at 0.67 0.236>,<at 0.984 0.0017 RCIM_Cubic>]}
        RedCurve {keys [<at 0.0514 8 RCIM_Cubic>,<at 0.67 2>,<at 0.984 0.1 RCIM_Cubic>]}
       }}))
    ScaleA=true
    ScaleRGB=true
   )
  ]
  rendererClasses [&NiagaraSpriteRendererProperties]
  renderers [{rendererClass=&NiagaraSpriteRendererProperties}]
 }
}
summary {
 systemName=NS_Ares_CombatEfficiency
 emitters [
  {emitterName=Flare, rendererClasses [&NiagaraSpriteRendererProperties]}
  {emitterName=Sparks, rendererClasses [&NiagaraSpriteRendererProperties]}
  {emitterName=Flames, rendererClasses [&NiagaraSpriteRendererProperties]}
  {emitterName=MuzzleHeat, rendererClasses [&NiagaraSpriteRendererProperties]}
  {emitterName=Glow, rendererClasses [&NiagaraSpriteRendererProperties]}
  {emitterName=ShockWave, rendererClasses [&NiagaraMeshRendererProperties]}
 ]
 userVariables []
}
